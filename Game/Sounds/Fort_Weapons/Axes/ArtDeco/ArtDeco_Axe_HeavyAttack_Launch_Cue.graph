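# Sound cue for the launch of the ArtDeco axe heavy attack, a player melee weapon sound.
# A quality-level root feeds two modulated branches and leaves its other two slots empty.
# Each branch randomly picks one of three equally weighted launch waves, with the second branch using a custom upward pitch range.
# Melee attenuation, sound class and concurrency settings govern playback, with attenuation overrides specifying log-reverse falloff and distance-based filtering.
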SoundCue /Game/Sounds/Fort_Weapons/Axes/ArtDeco/ArtDeco_Axe_HeavyAttack_Launch_Cue
root=QualityLevel_0
nodes [
 {id=Modulator_2, children [Random_5], properties {pitchMax=1.3, pitchMin=1}}
 {id=Modulator_4, children [Random_3]}
 {id=QualityLevel_0, children [Modulator_4,Modulator_2,null,null]}
 {id=Random_3, children [WavePlayer_13, WavePlayer_14, WavePlayer_15], properties {weights [1,1,1]}}
 {id=Random_5, children [WavePlayer_16, WavePlayer_17, WavePlayer_20], properties {weights [1,1,1]}}
 {id=WavePlayer_13, properties {soundWaveAssetPtr=&/Axe_ArtDeco_Heavy_Launch_01}}
 {id=WavePlayer_14, properties {soundWaveAssetPtr=&/Axe_ArtDeco_Heavy_Launch_02}}
 {id=WavePlayer_15, properties {soundWaveAssetPtr=&/Axe_ArtDeco_Heavy_Launch_03}}
 {id=WavePlayer_16, properties {soundWaveAssetPtr=&Axe_ArtDeco_Heavy_Launch_01}}
 {id=WavePlayer_17, properties {soundWaveAssetPtr=&Axe_ArtDeco_Heavy_Launch_02}}
 {id=WavePlayer_20, properties {soundWaveAssetPtr=&Axe_ArtDeco_Heavy_Launch_03}}
]
properties {
 attenuationSettings=&Fort_Weapon_Melee_Sml_Att
 maxDistance=2800
 soundClassObject=&Fort_WeaponMelee_PSM
 volumeMultiplier=0.8
 attenuationOverrides {
  bAttenuateWithLPF=true
  distanceAlgorithm=LogReverse
  falloffDistance=2500
  hPFFrequencyAtMax=250
  lPFFrequencyAtMax=1000
  lPFRadiusMax=25000
  lPFRadiusMin=0
  nonSpatializedRadiusStart=150
  reverbWetLevelMax=0.2
  reverbWetLevelMin=0.04
  stereoSpread=150
  attenuationShapeExtents {x=800}
 }
 concurrencySet [&Player_Melee_Weapon_Conc]
}
